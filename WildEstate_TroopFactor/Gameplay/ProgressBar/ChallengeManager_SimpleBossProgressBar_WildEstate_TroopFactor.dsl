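; Challenge manager for the Wild Estate TroopFactor encounter that drives a simple boss progress bar.
; Init calculates phase and stage progress targets, binds patrol-point-provider spawn results and updates area exit collision.
; Playlist readiness caches max team size, clamped 1–4, to index health and shield scalar curves.
; AI spawns trigger HandleBossHealthChanged, which sets BossProgress from scalar-weighted health and shield percentages once the boss is ready for its final fight.
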(event EventBeginPlay
  (bind (_gamestate _playlist) (Class|FortAsyncActionCurrentPlaylistReady|PlaylistDataReadyAsync)
    (:OnReady
      (Variables|Default|SetPlaylistMaxTeamSize (Math|Integer|Clamp(Integer) (Variables|/Script/FortniteGame.FortPlaylist|GetMaxTeamSize _playlist) 1 4)))))

(event Collision|EventActorBeginOverlap
  (params (OtherActor :type Actor%20Object%20Reference)))

(event EventTick
  (params (DeltaSeconds :type Float%20%28single-precision%29)))

(event EventInit
  (bind _patrolpointprovider (Variables|References|GetPatrolPointProvider))
  (Utilities|FlowControl|Sequence
    (:then_0
      (Progress|CalculatePhaseProgressTargets)
      (Progress|CalculateTotalProgressNeededPerStage)
      (Utilities|IsValid _patrolpointprovider
        (:Is%20Valid
          (EventDispatchers|BindEventtoOnActorSpawnedResult _patrolpointprovider (EventDispatchers|CreateEvent :!event AISpawned)))))
    (:then_1
      (CallFunction|UpdateAreaExitCollision))))

(event EventAISpawned
  (params (SpawnedActor :type Actor%20Object%20Reference) (bResult :type Boolean))
  (|CallParentFunction SpawnedActor bResult)
  (CallFunction|HandleBossHealthChanged))

(fn HandleBossHealthChanged
  (bind _bossactor (Variables|Boss|GetBossActor))
  (if (Variables|Default|GetBossIsReadyForFinalFight)
    (Utilities|IsValid _bossactor
      (:Is%20Valid
        (bind _asbp_player_pawn_non_participant_boss (Utilities|Casting|CastToBP_PlayerPawn_NonParticipant_Boss _bossactor))
        (Variables|Progress|SetBossProgress (Utilities|NormalizeFloat (+ (* (FortPawn|HealthPercent _asbp_player_pawn_non_participant_boss) (ScalableFloat|GetValueatLevel (Variables|Default|GetMaxHealthScalar) (Math|Conversions|ToFloat(Integer) (- (Variables|Default|GetPlaylistMaxTeamSize) 1)))) (* (FortPawn|ShieldPercent _asbp_player_pawn_non_participant_boss) (ScalableFloat|GetValueatLevel (Variables|Default|GetMaxShieldScalar) (Math|Conversions|ToFloat(Integer) (- (Variables|Default|GetPlaylistMaxTeamSize) 1))))) 0 (+ (ScalableFloat|GetValueatLevel (Variables|Default|GetMaxHealthScalar) (Math|Conversions|ToFloat(Integer) (- (Variables|Default|GetPlaylistMaxTeamSize) 1))) (ScalableFloat|GetValueatLevel (Variables|Default|GetMaxShieldScalar) (Math|Conversions|ToFloat(Integer) (- (Variables|Default|GetPlaylistMaxTeamSize) 1)))) true))))))

(fn ConstructionScript
  (|CallParentFunction))
