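# Translucent UI material for the TroopFactor progress bar that draws a confetti pattern of short line segments in a single tint color.
# Aspect-fitted UVs feed repeating, mirrored line-segment shapes, alongside a sine hash of floored coordinates for per-cell randomness.
# A FillChance threshold thins the pattern, with line and master opacity scaling the mask.
# A Progress parameter builds a hard horizontal cutoff gradient.
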Material /WildEstate_TroopFactor/Gameplay/ProgressBar/Materials/M_UI_TroopFactor_Confetti
graph {
 properties {bAutomaticallySetUsageInEditor=false, blendMode=BLEND_Translucent, materialDomain=MD_UI, usedShadingModels=Default Lit}
 nodes [
  Divide(A=ScalarParameter() {defaultValue=640, parameterName=Width}, B=ScalarParameter() {defaultValue=32, parameterName=Height})
  n1 = NamedRerouteUsage()
  n2 = ScalarParameter() {defaultValue=0.6, parameterName=FillChance}
  n3 = Multiply(
   A=ScalarParameter() {defaultValue=1, parameterName=Lines Opacity}
   B=Multiply(
    A=MaterialFunctionCall!(A=@n1, B=@n2) {materialFunction=/FortUILibrary/MaterialFunctions/Basics/MF_UI_BooleanOperations}.IsGreater
    B=NamedRerouteUsage()
   )
  )
  n4 = NamedRerouteUsage()
  NamedRerouteDeclaration(None=Multiply(A=Multiply(A=@n3, B=@n4), B=ScalarParameter() {defaultValue=1, parameterName=Master Opacity})) {name=Mask, nodeColor=<rgba 0 0 0 1>}
  n5 = NamedRerouteUsage()
  n6 = Sine(None=DotProduct(A=Floor(None=@n5), B=Constant2Vector() {r=141.4}))
  n7 = Constant() {r=0.5}
  NamedRerouteDeclaration(None=Frac(None=Multiply(A=@n6, B=@n7))) {name=Random2, nodeColor=<rgba 0 0.2031 1 1>}
  NamedRerouteDeclaration(None=VectorParameter() {parameterName=Col1, defaultValue=<rgba 0.0268 0.0311 0.0397 1>}.RGB) {name=Color, nodeColor=<rgba 0 0 0 1>}
  n8 = MaterialFunctionCall() {materialFunction=&GetUserInterfaceUV}
  NamedRerouteDeclaration(None=Divide(A=ComponentMask(None=@n8.Pixel Size) {r=true}, B=ComponentMask(None=@n8.Pixel Size) {g=true})) {name=AspectRatio, nodeColor=<rgba 1 0 0.6094 1>}
  NamedRerouteDeclaration(None=Add(
    A=MaterialFunctionCall() {materialFunction=&MF_UI_UVScaling}.Scaled to Fit
    B=Constant2Vector() {r=1.2}
   )) {name=UV_Fixed, nodeColor=<rgba 0 0.9297 1 1>}
  n9 = MaterialFunctionCall!(UV=Frac(None=NamedRerouteUsage()), Mirror Y=Constant(), Swizzle UV) {materialFunction=/FortUILibrary/MaterialFunctions/Basics/MF_UI_Symmetry_2x1_New}
  n10 = Constant2Vector() {g=1, r=0.6}
  NamedRerouteDeclaration(None=MaterialFunctionCall!(
    UVs=MaterialFunctionCall(UV=@n9.Flipped Symmetry, ContainerSize=@n10) {materialFunction=&MF_UI_UVScaling}.Scaled to Fit
    PointA=Constant() {r=-0.3}
    PointB
    Thickness=Constant() {r=0.2}
    OutlineThickness
    OutlineAlignment
    GlowMin
    GlowMax
   ) {materialFunction=/FortUILibrary/MaterialFunctions/Shapes/MF_UI_LineSegment}.Fill) {name=LIne, nodeColor=<rgba 0 0.4375 1 1>}
  n11 = Subtract(
   A=ComponentMask(None=MaterialFunctionCall() {materialFunction=&GetUserInterfaceUV}.Normalized UV) {r=true}
   B=ScalarParameter() {defaultValue=1, group=Progress, parameterName=Progress, sliderMax=1}
  )
  NamedRerouteDeclaration(None=OneMinus(None=Step(X=@n11) {constX=0})) {name=ProgressGradient, nodeColor=<rgba 0 0 0 1>}
  NamedRerouteUsage()
  NamedRerouteUsage()
 ]
}
