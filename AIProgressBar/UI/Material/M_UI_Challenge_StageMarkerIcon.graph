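# UI-domain translucent material for a stage marker icon, apparently for a challenge progress bar.
# It samples the shared phase icon texture and feeds RGBA into the IconToUse channel mask parameter, so one channel-packed texture can supply different icons.
# A Multiply node is also present, but the graph shows no links from it or the mask to material outputs.
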Material /AIProgressBar/UI/Material/M_UI_Challenge_StageMarkerIcon
graph {
 properties {blendMode=BLEND_Translucent, materialDomain=MD_UI, pixelDepthOffsetMode=PDOM_Legacy, usedShadingModels=Default Lit}
 nodes [
  ChannelMaskParameter(None=TextureSample() {texture=/AIProgressBar/UI/Texture/T_UI_Challenge_PhaseIcons}.RGBA) {parameterName=IconToUse}
  Multiply() {constA=1}
 ]
}
