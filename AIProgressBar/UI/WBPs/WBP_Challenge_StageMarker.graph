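# WBP_Challenge_StageMarker is a FortUserWidget for a challenge stage marker in the AIProgressBar UI, apparently a base for specific visuals.
# A fixed-size square SizeBox holds only a NamedSlot for those visuals.
# SetState and OnStageComplete are empty custom events, and the Tint Color variable is unused here.
# UpdateWidgetTextures only stores IsBossStage, while ForceUpdateVisibility applies the given visibility to the widget.
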WidgetBlueprint /AIProgressBar/UI/WBPs/WBP_Challenge_StageMarker
graphs [
 EventGraph {nodes [
   SetState!() -> (OutputDelegate, then, Active:bool=~, DelayOnAnimation:double=~) {is=title K2Node_CustomEvent}
   OnStageComplete() {is=title K2Node_CustomEvent}
  ]}
 ForceUpdateVisibility {nodes [
   n1 = ForceUpdateVisibility!() -> (then, Visibility:ESlateVisibility Enum=Visible) {is=title K2Node_FunctionEntry}
   > SetVisibility(InVisibility=@n1)
  ]}
 UpdateWidgetTextures {nodes [
   n1 = UpdateWidgetTextures!() -> (then, IsBossStage:bool=~) {is=title K2Node_FunctionEntry}
   > Set IsBossStage(IsBossStage:bool=@n1) -> (Output_Get:bool)
  ]}
]
variables [
 IsBossStage {type=boolean}
 Tint Color {type=SlateColor, default {colorUseRule=UseColor_Specified, specifiedColor=<rgba 0 1 0 1>}}
]
widgets {
 info {inheritedWidgetCount=0, parentClass=/Script/FortniteUI.FortUserWidget, rootWidgetClass=&SizeBox, widgetCount=2}
 widgets [
  SizeBox_Root {
   widgetClassPath=&SizeBox
   bIsVariable=true
   widget=SizeBox_Root
   properties {bOverride_HeightOverride=true, bOverride_WidthOverride=true, heightOverride=60, widthOverride=60}
  }
  NamedSlot_Visuals {widgetClassPath=&NamedSlot, bIsVariable=true, parent=SizeBox_Root, widget=NamedSlot_Visuals, slot {class=SizeBoxSlot}}
 ]
}
