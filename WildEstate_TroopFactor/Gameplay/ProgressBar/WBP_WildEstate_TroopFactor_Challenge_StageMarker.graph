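# TroopFactor stage marker widget for the challenge progress bar, extending the shared WBP_Challenge_StageMarker with an icon, shadow and arrow.
# UpdateWidgetTextures switches icon and shadow to the boss glyph through the IconToUse vector parameter, which acts as a channel selector.
# SetState stops animations and, when active, plays a delayed bounce with an arrow color fade.
# OnStageComplete plays the normal or boss completion animation, animating that selector and dimming the marker.
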WidgetBlueprint /WildEstate_TroopFactor/Gameplay/ProgressBar/WBP_WildEstate_TroopFactor_Challenge_StageMarker
animations [
 Anim_StageCompleted {
  binding_count=3
  length_seconds=0.2
  object=&self:Anim_StageCompleted
  sequence {bindings [
    Image_StageArrow {
     id=#1
     possessed_class=Image
     tracks [
      {
       type=MovieSceneColorTrack
       display_name=Color and Opacity
       sections [{
         type=MovieSceneColorSection
         end_frame=4
         start_frame=0
         channels [
          Color.R {type=MovieSceneScriptingFloatChannel, const [1,2]}
          Color.G {type=MovieSceneScriptingFloatChannel, keys [[0,0],[4,1]]}
          Color.B {type=MovieSceneScriptingFloatChannel, keys [[0,0],[4,1]]}
          Color.A {type=MovieSceneScriptingFloatChannel, const [1,1]}
         ]
        }]
      }
      {
       type=MovieSceneFloatTrack
       display_name=Render Opacity
       sections [{type=MovieSceneFloatSection, end_frame=4, start_frame=0, channels [None {type=MovieSceneScriptingFloatChannel, keys [[0,1],[4,0.5]]}]}]
      }
     ]
    }
    Image_StageIcon {
     id=#2
     possessed_class=Image
     tracks [
      {
       type=MovieSceneWidgetMaterialTrack
       sections [{
         type=MovieSceneParameterSection
         end_frame=4
         start_frame=0
         channels [
          IconToUseR {type=MovieSceneScriptingFloatChannel, keys [[0,1],[4,0]]}
          IconToUseG {type=MovieSceneScriptingFloatChannel, keys [[0,0],[4,1]]}
          IconToUseB {type=MovieSceneScriptingFloatChannel, const [0,1]}
          IconToUseA {type=MovieSceneScriptingFloatChannel, const [0,1]}
         ]
        }]
      }
      {
       type=MovieSceneFloatTrack
       display_name=Render Opacity
       sections [{type=MovieSceneFloatSection, end_frame=4, start_frame=0, channels [None {type=MovieSceneScriptingFloatChannel, keys [[0,1],[4,0.5]]}]}]
      }
     ]
    }
    Image_StageShadow {
     id=#3
     possessed_class=Image
     tracks [
      {
       type=MovieSceneWidgetMaterialTrack
       sections [{
         type=MovieSceneParameterSection
         end_frame=4
         start_frame=0
         channels [
          IconToUseR {type=MovieSceneScriptingFloatChannel, keys [[0,1],[4,0]]}
          IconToUseG {type=MovieSceneScriptingFloatChannel, keys [[0,0],[4,1]]}
          IconToUseB {type=MovieSceneScriptingFloatChannel, const [0,1]}
          IconToUseA {type=MovieSceneScriptingFloatChannel, const [0,1]}
         ]
        }]
      }
      {
       type=MovieSceneFloatTrack
       display_name=Render Opacity
       sections [{type=MovieSceneFloatSection, end_frame=4, start_frame=0, channels [None {type=MovieSceneScriptingFloatChannel, keys [[0,0.32],[2,0]]}]}]
      }
     ]
    }
   ]}
 }
 Anim_CurrentStage {
  binding_count=2
  length_seconds=0.45
  object=&self:Anim_CurrentStage
  sequence {bindings [
    Overlay_Images {
     id=#4
     display_name=Overlay_GodChallengeImages
     possessed_class=Overlay
     tracks [{
       type=MovieScene2DTransformTrack
       sections [{
         type=MovieScene2DTransformSection
         end_frame=9
         start_frame=0
         channels [
          Translation.X {type=MovieSceneScriptingFloatChannel, const [0,1]}
          Translation.Y {type=MovieSceneScriptingFloatChannel, keys [[0,-10],[3,-15],[6,5],[9,0]]}
         ]
        }]
      }]
    }
    Image_StageArrow {
     id=#5
     possessed_class=Image
     tracks [{
       type=MovieSceneColorTrack
       display_name=Color and Opacity
       sections [{
         type=MovieSceneColorSection
         end_frame=9
         start_frame=0
         channels [
          Color.R {type=MovieSceneScriptingFloatChannel, const [1,2]}
          Color.G {type=MovieSceneScriptingFloatChannel, keys [[3.4603,1],[9,0]]}
          Color.B {type=MovieSceneScriptingFloatChannel, keys [[3.4603,1],[9,0]]}
          Color.A {type=MovieSceneScriptingFloatChannel, const [1,1]}
         ]
        }]
      }]
    }
   ]}
 }
 Anim_BossStageCompleted {
  binding_count=3
  length_seconds=0.2
  object=&self:Anim_BossStageCompleted
  sequence {bindings [
    Image_StageShadow {
     id=#6
     possessed_class=Image
     tracks [
      {
       type=MovieSceneWidgetMaterialTrack
       sections [{
         type=MovieSceneParameterSection
         end_frame=4
         start_frame=0
         channels [
          IconToUseR {type=MovieSceneScriptingFloatChannel, keys [[0,1],[4,0]]}
          IconToUseG {type=MovieSceneScriptingFloatChannel, keys [[0,0],[4,1]]}
          IconToUseB {type=MovieSceneScriptingFloatChannel, const [0,1]}
          IconToUseA {type=MovieSceneScriptingFloatChannel, const [0,1]}
         ]
        }]
      }
      {
       type=MovieSceneFloatTrack
       display_name=Render Opacity
       sections [{type=MovieSceneFloatSection, end_frame=4, start_frame=0, channels [None {type=MovieSceneScriptingFloatChannel, keys [[0,0.32],[2,0]]}]}]
      }
     ]
    }
    Image_StageArrow {
     id=#7
     possessed_class=Image
     tracks [
      {
       type=MovieSceneColorTrack
       display_name=Color and Opacity
       sections [{
         type=MovieSceneColorSection
         end_frame=4
         start_frame=0
         channels [
          Color.R {type=MovieSceneScriptingFloatChannel, const [1,2]}
          Color.G {type=MovieSceneScriptingFloatChannel, keys [[0,0],[4,1]]}
          Color.B {type=MovieSceneScriptingFloatChannel, keys [[0,0],[4,1]]}
          Color.A {type=MovieSceneScriptingFloatChannel, const [1,1]}
         ]
        }]
      }
      {
       type=MovieSceneFloatTrack
       display_name=Render Opacity
       sections [{type=MovieSceneFloatSection, end_frame=4, start_frame=0, channels [None {type=MovieSceneScriptingFloatChannel, keys [[0,1],[4,0.5]]}]}]
      }
     ]
    }
    Image_StageIcon {
     id=#8
     possessed_class=Image
     tracks [
      {
       type=MovieSceneWidgetMaterialTrack
       sections [{
         type=MovieSceneParameterSection
         end_frame=4
         start_frame=0
         channels [
          IconToUseR {type=MovieSceneScriptingFloatChannel, const [0,2]}
          IconToUseG {type=MovieSceneScriptingFloatChannel, const [0,2]}
          IconToUseB {type=MovieSceneScriptingFloatChannel, const [1,1]}
          IconToUseA {type=MovieSceneScriptingFloatChannel, keys [[0,0],[4,1]]}
         ]
        }]
      }
      {
       type=MovieSceneFloatTrack
       display_name=Render Opacity
       sections [{type=MovieSceneFloatSection, end_frame=4, start_frame=0, channels [None {type=MovieSceneScriptingFloatChannel, keys [[0,1],[4,0.5]]}]}]
      }
     ]
    }
   ]}
 }
]
graphs [
 EventGraph {nodes [
   Event OnStageComplete() {is=title K2Node_Event}
   > PlayAnimation(InAnimation=Select(
     Option 0:Widget Animation Object Reference=Get Anim_StageCompleted() -> (Anim_StageCompleted:Widget Animation Object Reference)
     Option 1:Widget Animation Object Reference=Get Anim_BossStageCompleted() -> (Anim_BossStageCompleted:Widget Animation Object Reference)
     Index=Get IsBossStage() -> (IsBossStage:bool)
    ) -> (ReturnValue:Widget Animation Object Reference))
   n1 = Event SetState!() -> (OutputDelegate, then, Active:bool, DelayOnAnimation:double) {is=title K2Node_Event}
   > StopAllAnimations(self:User Widget Object Reference)
   > Branch(Condition=@n1.Active)
   > Delay(Duration=@n1.DelayOnAnimation)
   > PlayAnimation(InAnimation=Get Anim_CurrentStage() -> (Anim_CurrentStage:Widget Animation Object Reference))
  ]}
 UpdateWidgetTextures {nodes [
   n1 = UpdateWidgetTextures!() -> (then, IsBossStage:bool) {is=title K2Node_FunctionEntry}
   > Parent: UpdateWidgetTextures!(execute, IsBossStage:bool=@n1) -> (then) {is=title K2Node_CallParentFunction}
   > Branch(Condition=@n1)
   > n2 = GetDynamicMaterial(self=Get Image_StageIcon() -> (Image_StageIcon:Image Object Reference))
   n3 = SelectColor(A=(R=0,G=0,B=1,A=0), B=(R=1,G=0,B=0,A=0), bPickA=@n1)
   > SetVectorParameterValue(self=@n2, ParameterName=IconToUse, Value=@n3)
   > n4 = GetDynamicMaterial(self=Get Image_StageShadow() -> (Image_StageShadow:Image Object Reference))
   > SetVectorParameterValue(self=@n4, ParameterName=IconToUse, Value=@n3)
  ]}
]
variables []
widgets {
 info {
  inheritedWidgetCount=0
  namedSlotCount=1
  parentClass=/AIProgressBar/UI/WBPs/WBP_Challenge_StageMarker.WBP_Challenge_StageMarker_C
  rootWidgetClass=null
  widgetCount=4
 }
 widgets [
  Overlay_Images {
   widgetClassPath=&Overlay
   widget=Overlay_Images
   properties {renderOpacity=0, renderTransform {angle=0, scale=<1 1>, shear=<0 0>, translation=<0 -10>}, renderTransformPivot=<0.5 0>}
  }
  Image_StageShadow {
   widgetClassPath=&Image
   bIsVariable=true
   parent=Overlay_Images
   widget=Image_StageShadow
   properties {
    renderOpacity=0.336
    brush {
     bIsDynamicallyLoaded=false
     drawAs=Image
     imageType=FullColor
     mirroring=NoMirror
     resourceName=None
     resourceObject=/AIProgressBar/UI/Material/M_UI_Challenge_StageMarkerIcon
     tiling=NoTile
     imageSize=<56 56>
     margin=<ltrb 0 0 0 0>
     outlineSettings {
      bUseBrushTransparency=false
      roundingType=HalfHeightRadius
      width=0
      color {colorUseRule=UseColor_Specified, specifiedColor=<rgba 0 0 0 0>}
      cornerRadii=<0 0 0 0>
     }
     tintColor {colorUseRule=UseColor_Specified, specifiedColor=<rgba 0.099 0.099 0.099 1>}
     uVRegion {bIsValid=false, max=<0 0>, min=<0 0>}
    }
   }
   slot {class=OverlaySlot, properties {horizontalAlignment=HAlign_Center, padding=<ltrb 0 -7 0 0>}}
  }
  Image_StageIcon {
   widgetClassPath=&Image
   bIsVariable=true
   parent=Overlay_Images
   widget=Image_StageIcon
   properties {brush {
     bIsDynamicallyLoaded=false
     drawAs=Image
     imageType=FullColor
     mirroring=NoMirror
     resourceName=None
     resourceObject=&M_UI_Challenge_StageMarkerIcon
     tiling=NoTile
     imageSize=<55 55>
     margin=<ltrb 0 0 0 0>
     outlineSettings {
      bUseBrushTransparency=false
      roundingType=FixedRadius
      width=0
      color {colorUseRule=UseColor_Specified, specifiedColor=<rgba 0 0 0 1>}
      cornerRadii=<0 0 0 0>
     }
     tintColor {colorUseRule=UseColor_Specified, specifiedColor=<rgba 1 1 1 1>}
     uVRegion {bIsValid=false, max=<0 0>, min=<0 0>}
    }}
   slot {class=OverlaySlot, properties {horizontalAlignment=HAlign_Center, padding=<ltrb 0 -7 0 0>}}
  }
  Image_StageArrow {
   widgetClassPath=&Image
   bIsVariable=true
   parent=Overlay_Images
   widget=Image_StageArrow
   properties {
    brush {
     bIsDynamicallyLoaded=false
     drawAs=Image
     imageType=FullColor
     mirroring=NoMirror
     resourceName=None
     resourceObject=/AIProgressBar/UI/Material/M_UI_Challenge_MarkerArrow
     tiling=NoTile
     imageSize=<34 34>
     margin=<ltrb 0 0 0 0>
     outlineSettings {
      bUseBrushTransparency=false
      roundingType=HalfHeightRadius
      width=0
      color {colorUseRule=UseColor_Specified, specifiedColor=<rgba 0 0 0 0>}
      cornerRadii=<0 0 0 0>
     }
     tintColor {colorUseRule=UseColor_Specified, specifiedColor=<rgba 1 1 1 1>}
     uVRegion {bIsValid=false, max=<0 0>, min=<0 0>}
    }
    colorAndOpacity=<rgba 1 0 0 1>
   }
   slot {class=OverlaySlot, properties {horizontalAlignment=HAlign_Center, verticalAlignment=VAlign_Bottom}}
  }
 ]
}
